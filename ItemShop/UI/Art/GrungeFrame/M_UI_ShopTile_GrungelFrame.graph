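# Translucent UI material for item shop tiles that draws a grunge frame over a background built from shared background color parameters.
# A rounded-rect SDF, using the shop collection's CornerRadius, is min-combined with scaled, rotated grunge texture samples to form the frame mask.
# That mask drives color lerps feeding a Color Dodge blend, and a rounded-corners function shapes the tile.
# Named reroutes carry the linear gradient, fitted UVs and vertically weighted random noise overlay.
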Material /ItemShop/UI/Art/GrungeFrame/M_UI_ShopTile_GrungelFrame
graph {
 properties {bAutomaticallySetUsageInEditor=false, blendMode=BLEND_Translucent, materialDomain=MD_UI, usedShadingModels=Default Lit}
 nodes [
  MaterialFunctionCall!(Input Alpha=Constant() {r=1}, bOverrideCornerRadius, Corner Radius Override) {materialFunction=/FortUILibrary/Systems/ItemShopTile/Materials/MaterialFunctions/MF_UI_ItemShop_RoundedCorners}
  n1 = NamedRerouteUsage()
  n2 = NamedRerouteUsage()
  n3 = MaterialFunctionCall(UVs=@n2, Texture Scale=Constant() {r=0.65}) {materialFunction=&ScaleUVsByCenter}
  n4 = MaterialFunctionCall(UVs=@n2, Texture Scale=Constant() {r=1.1}) {materialFunction=&ScaleUVsByCenter}
  n5 = MaterialFunctionCall(UVs=@n4.UVs, "Rotation Angle (0-1)"=Constant() {r=0.33}) {materialFunction=&CustomRotator}
  n6 = TextureObject() {samplerType=SAMPLERTYPE_Normal, texture=/ItemShop/UI/Art/Jumpsuit/T_UI_Grunge1}
  n7 = Max(
   A=Saturate(None=TextureSample(UVs=@n5.Rotated Values, Tex=@n6) {samplerType=SAMPLERTYPE_Grayscale, texture=&/T_GrungeShavings_01}.R)
   B=Saturate(None=TextureSample(UVs=@n4.UVs, Tex=@n6) {samplerType=SAMPLERTYPE_Normal, texture=&T_UI_Grunge1}.R)
  )
  n8 = Constant() {r=0.5}
  n9 = CollectionParameter() {collection=/FortUILibrary/Systems/ItemShopTile/MPC_ItemShop, parameterName=CornerRadius}
  n10 = MaterialFunctionCall(R=@n9, G=@n9, B=@n9, A=@n9) {materialFunction=&AppendMany}
  n11 = Constant() {r=0.5}
  n12 = Min(
   A=SmoothStep(Value=MaterialFunctionCall(Size=@n8, Corner Radius=@n10.RGBA, Radius in Pixels=@n11) {materialFunction=&MF_UI_CoreUI_RoundedMask}.SDF) {constMax=0.8}
   B=Max(
    A=TextureSample(UVs=@n3.UVs) {samplerType=SAMPLERTYPE_Grayscale, texture=&T_GrungeShavings_01}.R
    B=Multiply(A=@n7) {constB=0.5}
   )
  )
  MaterialFunctionCall!(Blend=NamedRerouteUsage(), Base=LinearInterpolate(A=NamedRerouteUsage(), B=LinearInterpolate(B=@n1, Alpha=@n12) {constB=0}, Alpha=@n12)) {materialFunction=/Engine/Functions/Engine_MaterialFunctions03/Blends/Blend_ColorDodge}
  MaterialFunctionCall(UVs=@n3.UVs) {materialFunction=&CustomRotator}
  n13 = MaterialFunctionCall() {materialFunction=&MF_UI_UVScaling}
  NamedRerouteDeclaration(None=@n13.Scaled to Fit) {name=UV_Fit, nodeColor=<rgba 0 1 0.1094 1>}
  NamedRerouteDeclaration(None=@n13.ScaleY) {name=UV_ScaletoY, nodeColor=<rgba 1 0 0.7031 1>}
  n14 = NamedRerouteUsage()
  n15 = Constant()
  n16 = Constant() {r=1}
  n17 = Constant() {r=0.75}
  NamedRerouteDeclaration(None=LinearInterpolate(
    A=NamedRerouteUsage()
    B=NamedRerouteUsage()
    Alpha=MaterialFunctionCall(UVs=@n14, GradientStart=@n15, GradientEnd=@n16, RotationAngle=@n17) {materialFunction=&MF_UI_LinearGradient}.Result
   )) {name=GradientBase, nodeColor=<rgba 0.9453 1 0 1>}
  n18 = Multiply(
   A=MaterialFunctionCall!(UV=NamedRerouteUsage(), Seed) {materialFunction=&MF_UI_Random}.Result
   B=ScalarParameter() {defaultValue=0.3, group=Noise, parameterName=Noise Opacity}
  ) {constB=0.65}
  n19 = ComponentMask(None=NamedRerouteUsage()) {g=true}
  n20 = LinearInterpolate(Alpha=Power(Base=OneMinus(None=@n19))) {constA=0.1}
  NamedRerouteDeclaration(None=Saturate(None=Multiply(A=@n18, B=@n20))) {name=Noise Overlay, nodeColor=<rgba 0 0 0 1>}
  n21 = MaterialFunctionCall() {materialFunction=/FortUILibrary/Systems/ItemShopTile/Materials/MaterialFunctions/MF_UI_ItemShop_BackgroundParams}
  NamedRerouteDeclaration(None=@n21.Color 1) {name=color 1, nodeColor=<rgba 0.2422 0 1 1>}
  NamedRerouteDeclaration(None=@n21.Color 2) {name=color 2, nodeColor=<rgba 0.0781 1 0 1>}
  NamedRerouteDeclaration(None=@n21.Color 3) {name=color 3, nodeColor=<rgba 1 0 0.3984 1>}
  NamedRerouteUsage()
 ]
}
